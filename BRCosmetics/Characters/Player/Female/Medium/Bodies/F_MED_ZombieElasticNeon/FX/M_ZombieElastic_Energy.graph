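# Translucent, unlit, two-sided energy effect material for the ZombieElasticNeon character body FX, usable on Niagara sprites and mesh particles.
# Three color parameters blend along V and pass through emissive adjustments.
# An energy-circle mask samples UVs distorted by two panning noise textures, with distortion weighted along V.
# That mask is multiplied by sine bands, a V fade and a wipe driven by the ChangePercent dynamic parameter.
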Material /BRCosmetics/Characters/Player/Female/Medium/Bodies/F_MED_ZombieElasticNeon/FX/M_ZombieElastic_Energy
graph {
 properties {
  bUsedWithNiagaraMeshParticles=true
  bUsedWithNiagaraSprites=true
  blendMode=BLEND_Translucent
  pixelDepthOffsetMode=PDOM_Legacy
  shadingModel=MSM_Unlit
  twoSided=true
  usedShadingModels=Unlit
 }
 nodes [
  n1 = TextureCoordinate()
  MaterialFunctionCall(Emissive=MaterialFunctionCall(
    1=VectorParameter() {parameterName=Color, defaultValue=<rgba 2 1.1478 0.3589 1>}.RGB
    2=VectorParameter() {parameterName=Color2, defaultValue=<rgba 2 1.1478 0.3589 1>}.RGB
    3=VectorParameter() {parameterName=Color3, defaultValue=<rgba 2 1.1478 0.3589 1>}.RGB
    A=MaterialFunctionCall(Float2=@n1) {materialFunction=&BreakOutFloat2Components}.G
   ) {materialFunction=&Lerp_Multiple_Float3}.Lerp 3 Inputs) {materialFunction=&MF_FX_EmissiveAdjustments}
  n2 = Multiply(
   A=MaterialFunctionCall(Float2=TextureCoordinate()) {materialFunction=&BreakOutFloat2Components}.G
   B=ScalarParameter() {parameterName=TILESIZE}
  )
  n3 = Add(A=Abs(None=Sine(None=@n2)), B=ScalarParameter() {parameterName=AddT})
  n4 = MaterialFunctionCall(Float2=TextureCoordinate()) {materialFunction=&BreakOutFloat2Components}
  n5 = Saturate(None=Multiply(A=OneMinus(None=@n4.G)) {constB=3})
  n6 = VectorParameter() {parameterName=af, defaultValue=<rgba 2 1 0 1>}
  n7 = Multiply(A=Add(A=TextureCoordinate()) {constB=-0.5}, B=AppendVector(A=@n6.R, B=@n6.R))
  n8 = VectorParameter() {parameterName=offset, defaultValue=<rgba 0 0 0 1>}
  n9 = Power(Base=MaterialFunctionCall(Float2=TextureCoordinate()) {materialFunction=&BreakOutFloat2Components}.G)
  n10 = TextureCoordinate()
  n11 = VectorParameter() {parameterName=NoiseUV, defaultValue=<rgba 2 2 0 0.7>}
  n12 = AppendVector(A=@n11.R, B=@n11.R)
  n13 = AppendVector(A=@n11.B, B=@n11.B)
  n14 = Time() {bOverride_Period=true, period=600}
  n15 = TextureSampleParameter2D(UVs=Add(A=Multiply(A=@n10, B=@n12), B=Multiply(A=@n13, B=@n14))) {parameterName=Param, samplerType=SAMPLERTYPE_LinearGrayscale, texture=/Game/Global/Textures/Default/Flowmaps/T_StormSwirlMask}
  n16 = VectorParameter() {parameterName=AD, defaultValue=<rgba 0.2 -0.2 0 1>}
  n17 = Multiply(A=AppendVector(A=@n15.R, B=@n15.R), B=AppendVector(A=@n16.R, B=@n16.R))
  n18 = VectorParameter() {parameterName=Noise2UV, defaultValue=<rgba 1.5 1.5 0 1>}
  n19 = AppendVector(A=@n18.R, B=@n18.R)
  n20 = AppendVector(A=@n18.B, B=@n18.B)
  n21 = TextureSample(UVs=Add(A=Multiply(A=@n10, B=@n19), B=Multiply(A=@n20, B=@n14))) {samplerType=SAMPLERTYPE_Normal, texture=/Game/Gadgets/Assets/Telephone_Booth/Textures/T_BlindsNormalNoise_N}
  n22 = VectorParameter() {parameterName=AD2, defaultValue=<rgba 0.2 -0.2 0 1>}
  n23 = Multiply(A=AppendVector(A=@n21.R, B=@n21.R), B=AppendVector(A=@n22.R, B=@n22.R))
  n24 = Add(A=Add(A=Add(A=@n7) {constB=0.5}, B=AppendVector(A=@n8.R, B=@n8.R)), B=Multiply(A=OneMinus(None=@n9), B=Add(A=@n17, B=@n23)))
  n25 = Multiply(
   A=Multiply(
    A=TextureSample(UVs=@n24) {texture=&/T_EnergyCircle}.R
    B=MaterialFunctionCall!(TexCoord=@n24, ClampChannel) {materialFunction=/Game/Effects/Fort_Effects/MaterialFunctions/MF_FX_TilingMethod}.Result
   )
   B=ScalarParameter() {defaultValue=1}
  )
  n26 = TextureCoordinate()
  n27 = Multiply(A=Subtract(
    A=MaterialFunctionCall(Float2=@n26) {materialFunction=&BreakOutFloat2Components}.G
    B=DynamicParameter() {paramNames [Flick, Flick2, ChangePercent, Param4]}.ChangePercent
   )) {constB=15}
  Multiply(A=Multiply(A=Saturate(None=@n3), B=Multiply(A=@n5, B=@n25)), B=OneMinus(None=Saturate(None=@n27)))
  ScalarParameter() {defaultValue=0.5, parameterName=ssss}
 ]
}
