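# Material function that animates a texture as a grid flipbook, exposing color, alpha, cell UVs and cell center.
# Animation phase, optionally clamped, passes through frac, then floors against frame and row counts to pick a cell offset for the scaled input UVs.
# Both texture samples use shared clamp samplers, and a static switch selects mip bias or explicit mip level.
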MaterialFunction /BRCosmetics/Characters/Player/Female/Medium/Bodies/F_MED_ZombieElasticNeon/FX/FlipBook_Clamped
graph {
 properties {bExposeToLibrary=true, description="Animates a texture in a flipbook like fashion. "}
 nodes [
  n1 = FunctionInput(Preview=TextureCoordinate()) {
   bUsePreviewValueAsDefault=true
   description=Insert the texture coordinate of your choice.
   inputName=UVs
   inputType=FunctionInput_Vector2
   sortPriority=5
  }
  n2 = FunctionInput() {
   bUsePreviewValueAsDefault=true
   description=Enter the number of images horizontally across the texture.
   inputName=Number of Columns
   inputType=FunctionInput_Scalar
   previewValue=<2 0 0 1>
  }
  n3 = FunctionInput() {
   bUsePreviewValueAsDefault=true
   description=Enter the number of images vertically across the texture.
   inputName=Number of Rows
   inputType=FunctionInput_Scalar
   previewValue=<2 0 0 1>
  }
  n4 = AppendVector(A=@n2, B=@n3)
  n5 = Multiply(A=@n2, B=@n3)
  n6 = FunctionInput(Preview=Time()) {
   bUsePreviewValueAsDefault=true
   description="Apply a \"frac\" material node to the input value if time is used to control the animation phase."
   inputName="Animation  Phase (0-1)"
   inputType=FunctionInput_Scalar
  }
  n7 = FunctionInput(Preview=StaticBool()) {
   bUsePreviewValueAsDefault=true
   description="A frac operation is applied to the animation phase. This causes a value of exactly 1 in the animation phase input to return 0. Enabling this feature will clamp the animation phase value from 0 to .9999 which avoids the issue. "
   inputName="Clamp Anim (see tooltip)"
   inputType=FunctionInput_StaticBool
   sortPriority=100
  }
  n8 = Frac(None=StaticSwitch(True=Clamp(None=@n6) {maxDefault=1}, False=@n6, Value=@n7)) {desc=0-1 animation}
  n9 = Floor(None=Multiply(A=AppendVector(A=@n5, B=@n3), B=AppendVector(A=@n8, B=@n8)))
  n10 = Frac(None=Add(A=Divide(A=@n1, B=@n4), B=Divide(A=@n9, B=@n4)))
  n11 = FunctionInput(Preview=TextureObject() {texture=/Engine/Functions/Engine_MaterialFunctions02/ExampleContent/Textures/flipbook}) {bUsePreviewValueAsDefault=true, desc=Insert a texture object, inputName=Texture, inputType=FunctionInput_Texture2D}
  n12 = FunctionInput() {bUsePreviewValueAsDefault=true, inputName=MipBias/Level, inputType=FunctionInput_Scalar, sortPriority=99}
  n13 = TextureSample!(UVs=@n10, Tex=@n11, Bias=@n12, Apply View MipBias) {mipValueMode=TMVM_MipBias, samplerSource=SSM_Clamp_WorldGroupSettings}
  n14 = TextureSample!(UVs=@n10, Tex=@n11, Level=@n12, Apply View MipBias) {mipValueMode=TMVM_MipLevel, samplerSource=SSM_Clamp_WorldGroupSettings}
  n15 = FunctionInput(Preview=StaticBool() {value=true}) {
   bUsePreviewValueAsDefault=true
   description=Defaults to true.
   inputName="Use Mip Bias (T) Level (F)"
   inputType=FunctionInput_StaticBool
   sortPriority=100
  }
  FunctionOutput(None=StaticSwitch(True=@n13.RGB, False=@n14.RGB, Value=@n15))
  FunctionOutput(None=@n10) {outputName=UVs, sortPriority=1}
  FunctionOutput(None=StaticSwitch(True=@n13.A, False=@n14.A, Value=@n15)) {outputName=Alpha, sortPriority=1}
  n16 = Divide(A=Add(A=@n9) {constB=0.5}, B=@n4)
  MaterialFunctionCall!(
   Vector2=FunctionOutput(None=Frac(None=@n16)) {description=This value can be used for subsquent rotaiton operatons etc., outputName=UV Center, sortPriority=4}
   MaximumNumberOfDigits
   UVs
   DebugTextLocation RG_UpperRight BA_LowerLeft
   Component Spacing
  ) {materialFunction=/Engine/Functions/Engine_MaterialFunctions02/Utility/DebugFloat2Values}
 ]
}
