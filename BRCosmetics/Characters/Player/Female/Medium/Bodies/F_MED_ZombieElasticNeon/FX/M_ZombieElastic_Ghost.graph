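# Additive unlit, two-sided Niagara material for a ghostly hologram effect on the ZombieElastic Neon cosmetic, showing a still or flipbook sprite.
# A custom HLSL loop takes four noise-dithered alpha samples along Ghost_Direction, blending three tints through an injected Lerp3Color helper into a color-fringed smear.
# A particle-driven vertical threshold wipes tints to white and dims emissive, which particle color multiplies.
# A separate alpha sample, sine scanlines and particle alpha shape opacity.
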Material /BRCosmetics/Characters/Player/Female/Medium/Bodies/F_MED_ZombieElasticNeon/FX/M_ZombieElastic_Ghost
graph {
 properties {
  bUsedWithNiagaraMeshParticles=true
  bUsedWithNiagaraSprites=true
  blendMode=BLEND_Additive
  pixelDepthOffsetMode=PDOM_Legacy
  shadingModel=MSM_Unlit
  twoSided=true
  usedShadingModels=Unlit
 }
 nodes [
  n1 = ParticleColor()
  MaterialExpressionCustom_4 = Custom() {
   code="return 1;\r\n}\r\n\r\nhalf3 Lerp3Color (half Alpha,half3 R,half3 G,half3 B){\r\n  \r\nreturn lerp(   lerp(R,G,saturate(Alpha*2))   ,B,saturate(Alpha*2-1));\r\n"
   description=Lerp3ColorFunction
   outputType=CMOT_Float1
   inputs []
  }
  MaterialExpressionTextureObjectParameter_0 = TextureObjectParameter() {parameterName=2DHT, texture=/BRCosmetics/2dAssets/Emoji/Season22/S22_ChillCat_Claw_Sprite}
  MaterialExpressionConstant_4 = Constant() {r=4}
  n2 = ScalarParameter() {defaultValue=0.4, parameterName=Ghost_Distance}
  n3 = ScalarParameter() {defaultValue=4, parameterName=Columns, sliderMax=4}
  n4 = DynamicParameter() {parameterIndex=3, defaultValue=<rgba 1 0 1 1>, paramNames [AnimatedHologram, AnimationPhase, Param3, Param4]}
  MaterialExpressionLinearInterpolate_7 = LinearInterpolate(A=@n2, B=Divide(A=@n2, B=@n3), Alpha=@n4.AnimatedHologram) {desc=A}
  n5 = MaterialFunctionCall(UVs=TextureCoordinate(), Texture Scale=ScalarParameter() {defaultValue=0.75, parameterName=UV Scale}) {materialFunction=&ScaleUVsByCenter}
  n6 = AppendVector(A=ScalarParameter() {defaultValue=-0.14, parameterName=Ghost_OffsetX}, B=ScalarParameter() {parameterName=Ghost_OffsetY})
  n7 = ScalarParameter() {defaultValue=4, parameterName=Rows, sliderMax=4}
  n8 = MaterialFunctionCall!(
   "Animation  Phase (0-1)"=@n4.AnimationPhase
   Number of Rows=@n7
   Number of Columns=@n3
   Texture
   UVs=@n5.UVs
   MipBias/Level
   "Use Mip Bias (T) Level (F)"
   "Clamp Anim (see tooltip)"
  ) {materialFunction=&/FlipBook_Clamped}
  n9 = LinearInterpolate(A=@n6, B=Divide(A=@n6, B=AppendVector(A=@n3, B=@n7)), Alpha=@n4.AnimatedHologram) {desc=A}
  MaterialExpressionReroute_0 = Reroute(None=LinearInterpolate(A=Add(A=@n5.UVs, B=@n6), B=Add(A=@n8.UVs, B=@n9), Alpha=@n4.AnimatedHologram) {desc=A})
  MaterialExpressionVectorParameter_0 = VectorParameter() {parameterName=Ghost_Direction, defaultValue=<rgba 1 0 0 1>}
  n10 = Subtract(
   A=MaterialFunctionCall(Float2=TextureCoordinate()) {materialFunction=&BreakOutFloat2Components}.G
   B=DynamicParameter() {paramNames [Flick, Flick2, ChangePercent, Param4]}.ChangePercent
  )
  n11 = OneMinus(None=Saturate(None=Multiply(A=@n10) {constB=15}))
  MaterialExpressionLinearInterpolate_1 = LinearInterpolate(A=VectorParameter() {parameterName=Ghost_Color_R, defaultValue=<rgba 0 0.0192 1 1>}.RGB, Alpha=@n11)
  MaterialExpressionLinearInterpolate_3 = LinearInterpolate(A=VectorParameter() {parameterName=Ghost_Color_G, defaultValue=<rgba 0 1 0.4557 1>}.RGB, Alpha=@n11)
  MaterialExpressionLinearInterpolate_2 = LinearInterpolate(A=VectorParameter() {parameterName=Ghost_Color_B, defaultValue=<rgba 0 0.0786 1 1>}.RGB, Alpha=@n11)
  MaterialExpressionScreenPosition_0 = ScreenPosition()
  n12 = Custom!(ScreenPos=@MaterialExpressionScreenPosition_0.PixelPosition) {
   code="float random = InterleavedGradientNoise(ScreenPos.xy, float(View.StateFrameIndexMod8));\r\nreturn random;"
   outputType=CMOT_Float1
   inputs [{
     inputName=ScreenPos
     input {
      expression=&self:MaterialExpressionScreenPosition_0
      outputIndex=1
     }
    }]
  }
  MaterialExpressionSubtract_0 = Subtract(A=@n12) {constB=0.5}
  MaterialExpressionScalarParameter_7 = ScalarParameter() {parameterName=IsFrom3D}
  n13 = Custom!(
   Function=@MaterialExpressionCustom_4
   T2d=@MaterialExpressionTextureObjectParameter_0
   Amount=@MaterialExpressionConstant_4
   DistanceSum=@MaterialExpressionLinearInterpolate_7
   UV=@MaterialExpressionReroute_0
   Dir=@MaterialExpressionVectorParameter_0.RGB
   R=@MaterialExpressionLinearInterpolate_1
   G=@MaterialExpressionLinearInterpolate_3
   B=@MaterialExpressionLinearInterpolate_2
   Dither=@MaterialExpressionSubtract_0
   From3D=@MaterialExpressionScalarParameter_7
  ) {
   code="int IntAmount = floor(Amount);\r\nhalf PDistance = DistanceSum/IntAmount;\r\nhalf3 SumWeight = 0;\r\nhalf3 SumColor = 0;\r\nfor(int i = 0 ; i < IntAmount ; i++)\r\n    {\r\n        half3 ColorM = Lerp3Color(i/(half(IntAmount-1)),R,G,B);\r\n        half2 Pos = Dir*(PDistance*(i+Dither)) + UV;\r\n        half3 SampledColor = (Texture2DSample(T2d,GetMaterialSharedSampler(Material.Texture2D_0Sampler,View.MaterialTextureBilinearClampedSampler), Pos)).w;\r\n        SampledColor = lerp(SampledColor,1-SampledColor,From3D);\r\n        SumColor += SampledColor*ColorM;\r\n        SumWeight += ColorM;\r\n}\r\n\r\nreturn (SumColor/SumWeight);"
   inputs [
    {
     inputName=Function
     input {expression=&self:MaterialExpressionCustom_4}
    }
    {
     inputName=T2d
     input {expression=&self:MaterialExpressionTextureObjectParameter_0}
    }
    {
     inputName=Amount
     input {expression=&self:MaterialExpressionConstant_4}
    }
    {
     inputName=DistanceSum
     input {expression=&self:MaterialExpressionLinearInterpolate_7}
    }
    {
     inputName=UV
     input {expression=&self:MaterialExpressionReroute_0}
    }
    {
     inputName=Dir
     input {
      expression=&self:MaterialExpressionVectorParameter_0
      mask=1
      maskB=1
      maskG=1
      maskR=1
     }
    }
    {
     inputName=R
     input {expression=&self:MaterialExpressionLinearInterpolate_1}
    }
    {
     inputName=G
     input {expression=&self:MaterialExpressionLinearInterpolate_3}
    }
    {
     inputName=B
     input {expression=&self:MaterialExpressionLinearInterpolate_2}
    }
    {
     inputName=Dither
     input {expression=&self:MaterialExpressionSubtract_0}
    }
    {
     inputName=From3D
     input {expression=&self:MaterialExpressionScalarParameter_7}
    }
   ]
  }
  n14 = MaterialFunctionCall!(TexCoord=@n5.UVs, ClampChannel=Constant() {r=1}) {materialFunction=/Game/Effects/Fort_Effects/MaterialFunctions/MF_FX_TilingMethod}
  n15 = Multiply(A=LinearInterpolate(A=@n13, B=Multiply(A=@n13, B=@n14.Result), Alpha=@n4.AnimatedHologram) {desc=A}, B=ScalarParameter() {defaultValue=50})
  n16 = Multiply(A=@n15) {constB=0.7}
  MaterialFunctionCall(Emissive=Multiply(A=@n1.RGB, B=LinearInterpolate(A=@n15, B=@n16, Alpha=@n11))) {materialFunction=&MF_FX_EmissiveAdjustments}
  n17 = TextureCoordinate()
  n18 = Sine(None=Multiply(
    A=MaterialFunctionCall(Float2=@n17) {materialFunction=&BreakOutFloat2Components}.G
    B=ScalarParameter() {defaultValue=35, parameterName=ScanlineDensity}
   ))
  n19 = TextureSample(
   UVs=LinearInterpolate(
    A=@n5.UVs
    B=MaterialFunctionCall("Animation  Phase (0-1)"=@n4.AnimationPhase, Number of Rows=@n7, Number of Columns=@n3, UVs=@n5.UVs) {materialFunction=&FlipBook_Clamped}.UVs
    Alpha=@n4.AnimatedHologram
   ) {desc=A}
   Tex=@MaterialExpressionTextureObjectParameter_0
  ) {samplerSource=SSM_Clamp_WorldGroupSettings}
  n20 = LinearInterpolate(A=OneMinus(None=@n19.A), B=@n19.A, Alpha=@MaterialExpressionScalarParameter_7)
  n21 = ScalarParameter() {defaultValue=0.2, parameterName=Ghost_OpacityScale}
  Multiply(A=Multiply(A=Abs(None=@n18), B=Multiply(A=@n20, B=@n21)), B=Multiply(A=DynamicParameter().Param2, B=@n1.A))
 ]
}
