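# Translucent, two-sided Niagara sprite material for a glitch-edge effect on the female ZombieElasticNeon character.
# SubUV glitch noise is sampled at three horizontal offsets, split into RGB channels and tinted blue for an emissive adjustment.
# A mask multiplies a soft band, positioned along V by dynamic Param4, with flipbook-capable sprite alpha and desaturated noise that drops its brightest values.
# Param4 also fades that mask at both ends of its range.
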Material /BRCosmetics/Characters/Player/Female/Medium/Bodies/F_MED_ZombieElasticNeon/FX/M_ZombieElastic_GlitchEdge
graph {
 properties {bUsedWithNiagaraSprites=true, blendMode=BLEND_Translucent, pixelDepthOffsetMode=PDOM_Legacy, twoSided=true, usedShadingModels=Default Lit}
 nodes [
  Constant() {r=1}
  n1 = DynamicParameter() {paramNames [Fra,SIU,SIV,Param4]}
  n2 = MaterialFunctionCall!(U_Images=@n1.SIU, V_Images=@n1.SIU, SubUV Texture RGB, Time Speed=@n1.SIU, Manual Time Control, Time, UVs) {materialFunction=/Game/Effects/Fort_Effects/Materials/_Functions/MF_SubUV}
  n3 = MaterialFunctionCall!(OffsetAmount=ScalarParameter() {defaultValue=0.3, parameterName=GlitchOffset}) {materialFunction=/Engine/Functions/Engine_MaterialFunctions01/ImageAdjustment/BlurSampleOffsets}
  n4 = Constant()
  n5 = TextureSample(UVs=Add(
    A=@n2.UVs
    B=MaterialFunctionCall(X=@n3."1,0", Y=@n4) {materialFunction=&MakeFloat2}.Result
   )) {
   samplerType=SAMPLERTYPE_Masks
   texture=/Game/Characters/Player/Female/Medium/Bodies/F_MED_BlueBadass/Skins/NeonLines/F_MED_BlueBadass_NeonLines_-_GlitchNoise_v1
  }
  n6 = TextureSample(UVs=@n2.UVs) {
   samplerType=SAMPLERTYPE_Masks
   texture=&F_MED_BlueBadass_NeonLines_-_GlitchNoise_v1
  }
  n7 = TextureSample(UVs=Add(
    A=@n2.UVs
    B=MaterialFunctionCall(X=@n3."-1,0", Y=@n4) {materialFunction=&MakeFloat2}.Result
   )) {
   samplerType=SAMPLERTYPE_Masks
   texture=&F_MED_BlueBadass_NeonLines_-_GlitchNoise_v1
  }
  n8 = MaterialFunctionCall(X=@n5.R, Y=@n6.G, Z=@n7.B) {materialFunction=&MakeFloat3}
  MaterialFunctionCall(Emissive=Multiply(A=@n8.Result, B=Constant3Vector() {constant=<rgba 0.3086 0.6449 1 0>})) {materialFunction=&MF_FX_EmissiveAdjustments}
  MaterialFunctionCall!(Texture, UVs, SubImages, Frame=@n1.Fra) {materialFunction=/Game/Weapons/WeaponSkins/MaterialFunction/CustomNumbers_SubUV}
  n9 = MaterialFunctionCall(Float2=TextureCoordinate()) {materialFunction=&BreakOutFloat2Components}
  n10 = Abs(None=Multiply(A=Subtract(A=@n9.G, B=@n1.Param4)) {constB=1.3})
  n11 = Power(Base=OneMinus(None=Saturate(None=@n10))) {constExponent=5}
  n12 = ScalarParameter() {defaultValue=1.5, parameterName=affafafaf}
  n13 = TextureCoordinate()
  n14 = VectorParameter() {defaultValue=<rgba 1 1.3333 0 1>}
  n15 = Add(A=Multiply(A=Add(A=@n13) {constB=-0.5}, B=AppendVector(A=@n14.R, B=@n14.R))) {constB=0.5}
  n16 = DynamicParameter() {parameterIndex=3, defaultValue=<rgba 1 0 1 1>, paramNames [AnimatedHologram, AnimationPhase, Param3, Param4]}
  n17 = ScalarParameter() {defaultValue=4, parameterName=Rows, sliderMax=4}
  n18 = ScalarParameter() {defaultValue=4, parameterName=Columns, sliderMax=4}
  n19 = TextureSampleParameter2D(UVs=LinearInterpolate(
    A=@n15
    B=MaterialFunctionCall!(
     "Animation  Phase (0-1)"=@n16.AnimationPhase
     Number of Rows=@n17
     Number of Columns=@n18
     Texture
     UVs=@n15
     MipBias/Level
     "Use Mip Bias (T) Level (F)"
     "Clamp Anim (see tooltip)"
    ) {materialFunction=&/FlipBook_Clamped}.UVs
    Alpha=@n16.AnimatedHologram
   )) {parameterName=2DHT, samplerSource=SSM_Clamp_WorldGroupSettings, texture=/BRCosmetics/2dAssets/Emoji/Season22/S22_ChillCat_Claw_Sprite}
  n20 = Desaturation(None=@n8.Result)
  n21 = Constant() {r=1}
  n22 = Multiply(A=If(A=@n20, B=Constant() {r=0.9}, "A > B"=Constant(), "A == B"=@n21, "A < B"=@n21), B=@n20)
  n23 = Clamp(None=@n1.Param4) {maxDefault=0.05}
  n24 = Saturate(None=Multiply(A=Subtract(A=@n1.Param4) {constB=0.9}) {constB=10})
  Multiply(A=Multiply(A=Multiply(A=@n11, B=@n12), B=Multiply(A=@n19.A, B=@n22)), B=Multiply(A=Multiply(A=@n23) {constB=20}, B=OneMinus(None=@n24)))
  TextureCoordinate()
 ]
}
