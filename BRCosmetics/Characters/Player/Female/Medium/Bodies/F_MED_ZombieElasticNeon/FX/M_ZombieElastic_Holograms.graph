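# Translucent unlit two-sided hologram material for Niagara mesh particles in the ZombieElasticNeon character cosmetic, showing a 2D texture as a static or flipbook sprite.
# Emissive blends a cosine rainbow palette, tinted sprite color and a flickering, two-tone outline built from four offset alpha samples.
# Opacity combines sprite alpha, world-space scanlines, outline and optional strobe.
# Particle dynamic parameters drive flicker and a moving glitch band that also displaces vertices.
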Material /BRCosmetics/Characters/Player/Female/Medium/Bodies/F_MED_ZombieElasticNeon/FX/M_ZombieElastic_Holograms
graph {
 properties {
  bUsedWithNiagaraMeshParticles=true
  blendMode=BLEND_Translucent
  pixelDepthOffsetMode=PDOM_Legacy
  shadingModel=MSM_Unlit
  twoSided=true
  usedShadingModels=Unlit
 }
 nodes [
  n1 = MaterialFunctionCall(Float2=TextureCoordinate()) {materialFunction=&BreakOutFloat2Components}
  n2 = DynamicParameter() {paramNames [Flick, Flick2, ChangePercent, Param4]}
  n3 = Multiply(A=Subtract(A=@n1.G, B=@n2.ChangePercent)) {constB=15}
  n4 = OneMinus(None=Saturate(None=Abs(None=@n3)))
  n5 = Transform(None=LinearInterpolate(A=Constant3Vector() {constant=<rgba -1 0 0 0>}, B=Constant3Vector() {constant=<rgba 1 0 0 0>}, Alpha=@n1.R))
  n6 = Multiply(A=ScalarParameter() {defaultValue=30, parameterName=avawv}, B=DynamicParameter() {parameterIndex=1, paramNames [FEScale,0,0,0]}.FEScale)
  n7 = WorldPosition()
  n8 = ObjectPositionWS()
  n9 = Multiply(A=Normalize(VectorInput=Subtract(A=@n7.XYZ, B=@n8)), B=Constant3Vector() {constant=<rgba 1 1 0 0>})
  n10 = Add(
   A=ScalarParameter() {defaultValue=10, group=Noise, parameterName=WPOStrength}
   B=Multiply(A=ObjectRadius(), B=ScalarParameter() {defaultValue=0.025, group=Noise, parameterName=WPOStrengthFromObjectRadius}) {constB=0.05}
  )
  n11 = Divide(A=ObjectRadius(), B=ScalarParameter() {defaultValue=0.1, group=Noise, parameterName=NoiseSize})
  n12 = Multiply(A=Time(), B=VectorParameter() {group=Noise, parameterName=NoisePanSpeed, defaultValue=<rgba -0.08 0.07 -0.2 0>}.RGB)
  n13 = MaterialFunctionCall(
   TextureObject=TextureObject() {samplerType=SAMPLERTYPE_LinearColor, texture=&T_NoisePack_LinearColor}
   TextureSize=Multiply(A=VectorParameter() {group=Noise, parameterName=NoiseTileMult, defaultValue=<rgba 1 1 0.1 1>}.RGB, B=@n11)
   WorldPosition=Add(A=WorldPosition().XYZ, B=Multiply(A=@n11, B=@n12))
  ) {materialFunction=&WorldAlignedTexture}
  n14 = ActorPositionWS()
  n15 = MaterialFunctionCall(Float2=Round(None=Divide(A=@n14) {constB=2000})) {materialFunction=&BreakOutFloat2Components}
  n16 = Add(A=Time(), B=Add(A=@n15.R, B=@n15.R))
  n17 = Multiply(A=ScalarParameter() {defaultValue=0.5, parameterName=FlickerSpeed}, B=@n16)
  n18 = Sine(None=@n17) {period=1.6}
  n19 = Sine(None=@n17) {period=1.42}
  n20 = Ceil(None=AppendVector(A=AppendVector(A=@n18, B=@n19), B=Sine(None=@n17) {period=1.51}))
  n21 = ScalarParameter() {defaultValue=0.5, group=Noise, parameterName=NoiseFrequency}
  n22 = Add(A=Sine(None=@n17) {period=0.72}, B=Multiply(B=@n21) {constA=0.75}) {constB=0.75}
  n23 = Multiply(A=Add(A=Sine(None=@n17) {period=2.1}, B=Multiply(B=@n21) {constA=0.25}), B=@n22)
  n24 = Multiply(B=@n21) {constA=0.9, constB=0.9}
  n25 = Sine(None=@n17) {period=1.25}
  n26 = Multiply(A=@n22, B=Ceil(None=Add(A=@n24, B=@n25) {constB=0.9}))
  n27 = Sine(None=@n16) {period=0.075}
  n28 = Multiply(A=Saturate(None=Multiply(A=@n23, B=@n26)), B=LinearInterpolate(Alpha=ConstantBiasScale(None=@n27)) {constA=0.7})
  n29 = Add(
   A=ScalarParameter() {group=Noise, parameterName=NoiseAdd}
   B=Multiply(A=Multiply(A=@n20, B=@n28), B=ScalarParameter() {defaultValue=2, group=Noise, parameterName=NoiseMult})
  )
  n30 = Multiply(
   A=Multiply(A=Power(Base=@n13.XY Texture) {constExponent=4}, B=ScalarParameter() {defaultValue=2, group=Noise, parameterName=NoiseMaskMult}) {constB=20}
   B=@n29
  )
  n31 = ComponentMask(None=MaterialFunctionCall() {materialFunction=&ObjectPivotPoint}.Object Pivot Location) {b=true}
  n32 = ScalarParameter() {defaultValue=300, parameterName=PulseHeight}
  n33 = ScalarParameter() {defaultValue=1, parameterName=TurnOn}
  n34 = ScalarParameter() {defaultValue=33, parameterName=PulseRadius}
  n35 = SphereMask(
   A=ComponentMask(None=WorldPosition().XYZ) {b=true}
   B=LinearInterpolate(A=@n31, B=Add(A=@n31, B=@n32) {constB=1600}, Alpha=@n33) {constA=-0.25, constB=1.2}
   Radius=LinearInterpolate(A=@n34, B=Multiply(A=@n34) {constB=0.1}, Alpha=@n33) {constA=500, constB=100}
  ) {hardnessPercent=0}
  n36 = OneMinus(None=@n33)
  n37 = Add(A=Desaturation(None=Saturate(None=@n30)), B=Multiply(A=Multiply(A=@n35, B=@n36)) {constB=2})
  Add(A=Multiply(A=Multiply(A=@n4, B=@n5), B=Multiply(A=@n6, B=@n2.Flick2)), B=Multiply(A=@n37, B=Multiply(A=@n9, B=@n10)) {constB=0})
  Frac(None=Time())
  n38 = Constant()
  LinearInterpolate(Alpha=If(A=ScalarParameter() {defaultValue=1, parameterName=FrameCount, sliderMax=16}, "A > B"=Constant() {r=1}, "A == B"=@n38, "A < B"=@n38) {constB=1})
  n39 = Multiply(A=Multiply(A=Power(Base=@n4) {constExponent=5}, B=ScalarParameter() {defaultValue=1.5, parameterName=affafafaf}), B=Constant() {r=1})
  n40 = Time()
  n41 = VectorParameter() {group=Rainbow, parameterName=wadWFAWFG}
  n42 = DotProduct(A=CameraVectorWS(), B=PixelNormalWS())
  n43 = Add(
   A=Multiply(A=Time(), B=ScalarParameter() {defaultValue=0.1, group=Rainbow, parameterName=TimeScaleOnRef})
   B=ScalarParameter() {group=Rainbow, parameterName=Offset}
  )
  n44 = Multiply(
   A=MaterialFunctionCall(Hue Shift Percentage=Multiply(A=@n40, B=@n41.G), Texture=VectorParameter() {group=Rainbow, parameterName=C, defaultValue=<rgba 1 1 1 1>}.RGB) {materialFunction=&HueShift}.Result
   B=Multiply(A=Add(A=@n42, B=@n43), B=ScalarParameter() {defaultValue=7, group=Rainbow, parameterName=ssw'd}) {constB=2}
  )
  n45 = MaterialFunctionCall(
   Hue Shift Percentage=Multiply(A=@n40, B=@n41.R)
   Texture=VectorParameter() {group=Rainbow, parameterName=D, defaultValue=<rgba 0 0.33 0.67 1>}.RGB
  ) {materialFunction=&HueShift}
  n46 = Multiply(A=VectorParameter() {group=Rainbow, parameterName=B, defaultValue=<rgba 0.5 0.5 0.5 1>}.RGB, B=Cosine(None=Add(A=@n44, B=@n45.Result)))
  n47 = VectorParameter() {group=Rainbow, parameterName=A, defaultValue=<rgba 0.5 0.5 0.5 1>}
  n48 = Multiply(
   A=Desaturation(None=Add(A=@n47.RGB, B=@n46), Fraction=ScalarParameter() {group=Rainbow, parameterName=Desaturation2})
   B=ScalarParameter() {defaultValue=0.3, group=Rainbow, parameterName=ReS}
  )
  n49 = TextureCoordinate()
  n50 = ScalarParameter() {defaultValue=0.75, parameterName=UV Scale}
  n51 = MaterialFunctionCall(UVs=@n49, Texture Scale=@n50) {materialFunction=&ScaleUVsByCenter}
  n52 = TextureObjectParameter() {group=Texture, parameterName=2DHologramTexture, texture=/BRCosmetics/2dAssets/Emoji/Season22/S22_ChillCat_Claw_Sprite}
  n53 = TextureSample(UVs=@n51.UVs, Tex=@n52) {samplerSource=SSM_Clamp_WorldGroupSettings, texture=/BRCosmetics/2dAssets/Emoji/Season4/babySeal}
  n54 = ScalarParameter() {parameterName=AnimationPhase}
  n55 = ScalarParameter() {defaultValue=4, parameterName=Rows, sliderMax=4}
  n56 = ScalarParameter() {defaultValue=4, parameterName=Columns, sliderMax=4}
  n57 = MaterialFunctionCall!(
   "Animation  Phase (0-1)"=@n54
   Number of Rows=@n55
   Number of Columns=@n56
   Texture=@n52
   UVs=@n51.UVs
   MipBias/Level
   "Use Mip Bias (T) Level (F)"
   "Clamp Anim (see tooltip)"
  ) {materialFunction=&/FlipBook_Clamped}
  n58 = ScalarParameter() {defaultValue=1, parameterName=AnimatedHologram}
  n59 = LinearInterpolate(A=@n53.RGB, B=@n57.Result, Alpha=@n58) {desc=RGB}
  n60 = LinearInterpolate(A=Multiply(A=@n48) {constB=0.07}, B=@n48, Alpha=Desaturation(None=@n59))
  n61 = ScalarParameter() {defaultValue=1, group=Noise, parameterName=NoiseFresnelXPMult}
  n62 = QualitySwitch(
   Default=MaterialFunctionCall(In=@n37, Contrast=ScalarParameter() {defaultValue=0.2, group=Noise, parameterName=NoiseContrast}) {materialFunction=&CheapContrast}.Result
   Low=Constant()
  )
  n63 = Power(
   Base=Desaturation(None=@n59, Fraction=ScalarParameter() {defaultValue=0.3, group=Texture, parameterName=Desaturation})
   Exp=Multiply(A=ScalarParameter() {defaultValue=2, parameterName=FresnelExp}, B=LinearInterpolate(B=@n61, Alpha=@n62) {constA=1, constB=0.5})
  )
  n64 = Multiply(A=@n63, B=ScalarParameter() {defaultValue=1, parameterName=EmissiveMult}) {constB=2}
  n65 = MaterialFunctionCall(
   Hue Shift Percentage=ScalarParameter() {defaultValue=-0.05, parameterName="Hue Shift Percentage (S)"}
   Texture=Desaturation(None=@n64, Fraction=ScalarParameter() {defaultValue=0.2, parameterName=awf})
  ) {materialFunction=&HueShift}
  n66 = VectorParameter() {parameterName=EmissiveColorMult, defaultValue=<rgba 0.6 1.5333 2 1>}
  n67 = ScalarParameter() {defaultValue=6, group=ZombieElasticPost}
  n68 = Multiply(A=MaterialFunctionCall(Float2=@n51.UVs) {materialFunction=&BreakOutFloat2Components}.G) {constB=0.5}
  n69 = Power(Base=OneMinus(None=Sine(None=@n68)), Exp=ScalarParameter() {defaultValue=1, group=ZombieElasticPost, parameterName=Param_7})
  n70 = Desaturation(None=@n64, Fraction=Multiply(A=@n69, B=ScalarParameter() {defaultValue=-1.5, parameterName=Deastawst}))
  n71 = OneMinus(None=Desaturation(None=Saturate(None=@n29)))
  n72 = Saturate(None=Add(A=@n71, B=ScalarParameter() {defaultValue=0.25, parameterName=ffafawfaffff}) {constB=0.25})
  n73 = Multiply(
   A=DynamicParameter() {paramNames [Flick, Flick2, ChangePercent, Param4]}.Flick2
   B=LinearInterpolate(A=Multiply(A=@n65.Result, B=@n66.RGB), B=Multiply(A=@n67, B=@n70), Alpha=Multiply(A=@n69, B=@n72))
  )
  n74 = VectorParameter() {parameterName=L, defaultValue=<rgba 0 0.5 0 1>}
  n75 = VectorParameter() {parameterName=R, defaultValue=<rgba 0.5 0 0 1>}
  n76 = TextureCoordinate()
  n77 = Constant2Vector() {g=0.5, r=0.5}
  n78 = MaterialFunctionCall(
   In=MaterialFunctionCall(Float2=MaterialFunctionCall(UVs=@n76, Texture Scale=@n77) {materialFunction=&ScaleUVsByCenter}.UVs) {materialFunction=&BreakOutFloat2Components}.R
   Contrast=Constant() {r=0.5}
  ) {materialFunction=&CheapContrast}
  n79 = MaterialFunctionCall!(OffsetAmount=ScalarParameter() {defaultValue=0.03, parameterName=OutLine}) {materialFunction=/Engine/Functions/Engine_MaterialFunctions01/ImageAdjustment/BlurSampleOffsets}
  n80 = Add(A=@n51.UVs, B=Multiply(A=@n79."0,1") {constB=0.17})
  n81 = MaterialFunctionCall("Animation  Phase (0-1)"=@n54, Number of Rows=@n55, Number of Columns=@n56, Texture=@n52, UVs=@n80) {materialFunction=&FlipBook_Clamped}
  n82 = MaterialFunctionCall(UVs=@n49, Texture Scale=Add(A=@n50, B=Constant() {r=-0.04}) {constB=0}) {materialFunction=&ScaleUVsByCenter}
  n83 = LinearInterpolate(A=@n53.A, B=Min(A=@n57.Alpha, B=@n82.0-1 mask), Alpha=@n58) {desc=A}
  n84 = Subtract(
   A=LinearInterpolate(A=TextureSample(UVs=@n80, Tex=@n52) {samplerSource=SSM_Clamp_WorldGroupSettings}.A, B=Min(A=@n81.Alpha, B=@n82.0-1 mask), Alpha=@n58) {desc=A}
   B=@n83
  )
  n85 = Desaturation(None=Saturate(None=@n29))
  n86 = ScalarParameter() {defaultValue=0.17, parameterName=gegegeg}
  n87 = Multiply(A=Saturate(None=Add(A=@n85, B=@n86) {constB=0.25})) {constB=2}
  n88 = Add(A=@n51.UVs, B=Multiply(A=@n79."1,0", B=@n87))
  n89 = MaterialFunctionCall("Animation  Phase (0-1)"=@n54, Number of Rows=@n55, Number of Columns=@n56, Texture=@n52, UVs=@n88) {materialFunction=&FlipBook_Clamped}
  n90 = Subtract(
   A=LinearInterpolate(A=TextureSample(UVs=@n88, Tex=@n52) {samplerSource=SSM_Clamp_WorldGroupSettings}.A, B=Min(A=@n89.Alpha, B=@n82.0-1 mask), Alpha=@n58) {desc=A}
   B=@n83
  )
  n91 = Add(A=@n51.UVs, B=Multiply(A=@n79."0,-1") {constB=0.17})
  n92 = LinearInterpolate(
   A=TextureSample(UVs=@n91, Tex=@n52) {samplerSource=SSM_Clamp_WorldGroupSettings}.A
   B=Min(
    A=MaterialFunctionCall("Animation  Phase (0-1)"=@n54, Number of Rows=@n55, Number of Columns=@n56, Texture=@n52, UVs=@n91) {materialFunction=&FlipBook_Clamped}.Alpha
    B=@n82.0-1 mask
   )
   Alpha=@n58
  ) {desc=A}
  n93 = Add(A=@n51.UVs, B=Multiply(A=@n79."-1,0", B=@n87))
  n94 = TextureSample(UVs=@n93, Tex=@n52) {samplerSource=SSM_Clamp_WorldGroupSettings}
  n95 = Min(
   A=MaterialFunctionCall("Animation  Phase (0-1)"=@n54, Number of Rows=@n55, Number of Columns=@n56, Texture=@n52, UVs=@n93) {materialFunction=&FlipBook_Clamped}.Alpha
   B=@n82.0-1 mask
  )
  n96 = Max(A=Max(A=Max(A=@n84, B=@n90), B=Subtract(A=@n92, B=@n83)), B=Subtract(A=LinearInterpolate(A=@n94.A, B=@n95, Alpha=@n58) {desc=A}, B=@n83))
  n97 = Saturate(None=QualitySwitch(Default=@n96, Low=Constant(), High=@n96, Medium=@n96, Epic=@n96))
  n98 = ScalarParameter() {defaultValue=-0.6, parameterName=ffafawfaffff2}
  n99 = LinearInterpolate(
   A=Multiply(A=LinearInterpolate(A=@n74.RGB, B=@n75.RGB, Alpha=@n78.Result), B=@n97)
   B=@n97
   Alpha=Saturate(None=Add(A=@n71, B=@n98) {constB=0.25})
  )
  n100 = ScalarParameter() {defaultValue=3, parameterName=Param_2}
  n101 = Saturate(None=Subtract(A=@n59, B=ScalarParameter() {defaultValue=-0.1, parameterName=twd}))
  n102 = Power(Base=OneMinus(None=Desaturation(None=@n101)), Exp=ScalarParameter() {defaultValue=4, parameterName=In_Contrast})
  n103 = Saturate(None=@n29)
  n104 = Add(A=OneMinus(None=Desaturation(None=@n103)), B=ScalarParameter() {defaultValue=0.4, parameterName=awd}) {constB=0.25}
  n105 = Multiply(
   A=DynamicParameter() {paramNames [Flick, Flick2, ChangePercent, Param4]}.Flick
   B=Multiply(A=Saturate(None=@n104), B=ScalarParameter() {defaultValue=20, parameterName=sca})
  )
  n106 = LinearInterpolate(
   A=LinearInterpolate(A=Add(A=@n60, B=@n73), B=Multiply(A=@n99, B=@n100), Alpha=@n97)
   B=Add(A=@n97, B=Multiply(A=@n102, B=@n105))
   Alpha=OneMinus(None=Saturate(None=@n3))
  )
  n107 = ParticleColor()
  MaterialFunctionCall(Emissive=Multiply(A=Add(A=@n39, B=@n106), B=@n107.RGB)) {materialFunction=&MF_FX_EmissiveAdjustments}
  n108 = TextureObjectParameter() {group=Texture, parameterName=2DHologramTexture, texture=&S22_ChillCat_Claw_Sprite}
  Multiply(A=TextureSample(UVs=@n80, Tex=@n108) {samplerSource=SSM_Clamp_WorldGroupSettings}.A)
  n109 = Multiply(A=@n83, B=@n82.0-1 mask)
  n110 = Saturate(None=@n62)
  n111 = Add(A=@n63, B=Multiply(A=ScalarParameter() {group=Noise, parameterName=NoiseBrightness}, B=@n110) {constB=0.1})
  n112 = Multiply(A=@n62) {constB=1.5}
  n113 = Subtract(A=ScalarParameter() {defaultValue=0.2, parameterName=MinScanlineOpacity}, B=@n112)
  n114 = Ceil(None=@n110)
  n115 = WorldPosition()
  n116 = ScalarParameter() {defaultValue=20000, parameterName=ScaleDistance}
  n117 = CameraPositionWS()
  n118 = ActorPositionWS()
  n119 = Saturate(None=Divide(A=Distance(A=@n117, B=@n118), B=@n116) {constB=1000})
  n120 = Multiply(
   A=Multiply(A=ScalarParameter() {defaultValue=400, parameterName=ScanlineTile}, B=LinearInterpolate(Alpha=@n114) {constA=1, constB=1.25})
   B=Divide(A=ComponentMask(None=@n115.XYZ) {b=true}, B=LinearInterpolate(B=@n116, Alpha=@n119) {constA=50, constB=100}) {constB=20}
  ) {constA=1}
  n121 = DotProduct(A=VertexNormalWS(), B=Constant3Vector() {constant=<rgba 0 0 1 0>})
  n122 = LinearInterpolate(A=ConstantBiasScale(None=Sine(None=@n120)), Alpha=Power(Base=Abs(None=@n121)) {constExponent=5}) {constB=0.25}
  n123 = Multiply(
   A=LinearInterpolate(A=Multiply(A=@n109, B=@n111), B=@n109) {constAlpha=1}
   B=LinearInterpolate(A=Clamp(None=@n113) {minDefault=-0.25}, B=Add(A=@n112), Alpha=Power(Base=@n122) {constExponent=1.5}) {constA=0.5}
  )
  n124 = Multiply(
   A=LinearInterpolate(B=@n107.A, Alpha=ScalarParameter() {parameterName=UseParticleColor}) {constA=1, desc=Workaround for skel mesh bug}
   B=Multiply(A=Saturate(None=@n123), B=ScalarParameter() {defaultValue=1.2, parameterName=OpacityMult})
  )
  n125 = Sine(None=Multiply(A=Time(), B=ScalarParameter() {defaultValue=5, parameterName=StrobeSpeed}))
  n126 = ScalarParameter() {defaultValue=2, parameterName=StrobeHardness}
  n127 = LinearInterpolate(B=Clamp(None=Multiply(A=@n125, B=@n126)), Alpha=ScalarParameter() {parameterName=Strobe}) {constA=1}
  Saturate(None=Add(A=Subtract(A=@n97) {constB=0.01}, B=Multiply(A=@n124, B=@n127)))
  ScalarParameter() {defaultValue=0.5, parameterName=P}
  TextureSample(UVs=@n88, Tex=@n108) {samplerSource=SSM_Clamp_WorldGroupSettings}
  TextureSample(UVs=@n91, Tex=@n108) {samplerSource=SSM_Clamp_WorldGroupSettings}
  TextureSample(UVs=@n93, Tex=@n108) {samplerSource=SSM_Clamp_WorldGroupSettings}
  TextureSample(UVs=@n51.UVs, Tex=@n108) {samplerSource=SSM_Clamp_WorldGroupSettings}
 ]
}
