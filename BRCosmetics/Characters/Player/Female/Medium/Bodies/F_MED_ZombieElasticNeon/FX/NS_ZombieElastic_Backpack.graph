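# Niagara backpack effect for the female ZombieElasticNeon body cosmetic, layering a mesh "Sticker" over or among three fixed-orientation "Darken" sprites.
# Each emitter bursts one particle, offset from the simulation position and toward camera.
# Dynamic material parameters carry age-driven waveform flicker, hologram animation inputs and a KillingChange value.
# Front-end preview substitutes System.AutoSwitch for KillingChange and halves the Sticker's scale parameter, while User.Boost scales Sticker color.
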NiagaraSystem /BRCosmetics/Characters/Player/Female/Medium/Bodies/F_MED_ZombieElasticNeon/FX/NS_ZombieElastic_Backpack
dependencies {
 usedDataInterfaces []
 usedDynamicInputs [&Sine]
 usedModules [
  &SystemState
  &self:SystemSpawnScript.SystemScriptSource.SystemScriptGraph.NiagaraNodeAssignment_0.SetVariables_#1
 ]
}
emitters {
 Darken {
  emitterUpdateScript [
   &EmitterState(Life Cycle Mode=::NewEnumerator1, Loop Behavior=::NewEnumerator1, Loop Duration Mode=::NewEnumerator1, UseLoopDelay=false)
   &SpawnBurst_Instantaneous(Spawn Count=1, Use Spawn Probability=false)
  ]
  particleSpawnScript [
   /Niagara/Modules/Spawn/Initialization/InitializeParticle(
    Color=<rgba 0 0 0 0.45>
    Module.Write Color=true
    Module.Write Lifetime=false
    Module.Write Mass=true
    Module.Write Material Random=true
    Module.Write Position=true
    Module.Write Scale=false
    Module.Write SpriteRotation=false
    Module.Write SpriteSize=true
    Position=&Add_Vector(A=/Niagara/DynamicInputs/Helpers/SimulationPosition(), B=<0 -8.5 0>)
    Sprite Size=<35 35>
   )
   &DynamicMaterialParameters(
    Index 0 Param 1=1.2
    Index 0 Param 2=0.15
    Index 0 Param 3=1
    Index 0 Param 4=1
    Module.Param0WriteEnabled=true
    Module.Param1WriteEnabled=true
    Module.Param2WriteEnabled=true
    Module.Param3WriteEnabled=true
   )
   &CameraOffset(Camera Offset Amount=10)
   /NiagaraFN/Sprite/FN_SpriteAlignmentAndFacing(SpriteAlignment:Vector3f=<0 0 1>, SpriteFacing:Vector3f=<0 -1 0>) {module=SpriteAlignmentAndFacing}
  ]
  particleUpdateScript [&ParticleState(Kill Particles When Lifetime Has Elapsed=false, Let Infinitely Lived Particles Die When Emitter Deactivates=true)]
  rendererClasses [&NiagaraSpriteRendererProperties]
  renderers [{rendererClass=&NiagaraSpriteRendererProperties}]
 }
 Darken001 {
  emitterUpdateScript [
   &EmitterState(Life Cycle Mode=::NewEnumerator1, Loop Behavior=::NewEnumerator1, Loop Duration Mode=::NewEnumerator1, UseLoopDelay=false)
   &SpawnBurst_Instantaneous(Spawn Count=1, Use Spawn Probability=false)
  ]
  particleSpawnScript [
   &InitializeParticle(
    Color=<rgba 0 0 0 1>
    Module.Write Color=true
    Module.Write Lifetime=false
    Module.Write Mass=true
    Module.Write Material Random=true
    Module.Write Position=true
    Module.Write Scale=false
    Module.Write SpriteRotation=false
    Module.Write SpriteSize=true
    Position=&Add_Vector(A=&SimulationPosition(), B=<0 -8.5 5>)
    Sprite Size=<35 30>
   )
   &CameraOffset(Camera Offset Amount=10)
   &FN_SpriteAlignmentAndFacing(SpriteAlignment:Vector3f=<0 0 1>, SpriteFacing:Vector3f=<0 -1 0>) {module=SpriteAlignmentAndFacing}
  ]
  particleUpdateScript [
   &ParticleState(Kill Particles When Lifetime Has Elapsed=false, Let Infinitely Lived Particles Die When Emitter Deactivates=true)
   &DynamicMaterialParameters(
    Index 0 Param 1=&Waveform(
     "Waveform [1]"=::NewEnumerator8
     "[1] Phase Scale"=20
     "[1] Frequency"=1
     "[1] Amplitude Min/Max"=<0 1>
     "[1] Scale"=0.9
     "[1] Bias"=0.1
     "[1] Exponent"=1.3
     "[2] Phase"=$Emitter.Age
     "[3] Phase"=$Emitter.Age
    )
    Index 0 Param 2=&Waveform(
     "Waveform [1]"=::NewEnumerator8
     "[1] Phase Scale"=25
     "[1] Frequency"=1
     "[1] Amplitude Min/Max"=<0 1>
     "[1] Scale"=0.7
     "[1] Bias"=0.3
     "[2] Phase"=$Emitter.Age
     "[3] Phase"=$Emitter.Age
    )
    Index 0 Param 3=&MakeCustomFloatFromBool(Bool=$User.bIsFrontEndPreview, True Float=$System.AutoSwitch, False Float=$User.KillingChange)
    Module.Param0WriteEnabled=true
    Module.Param1WriteEnabled=true
    Module.Param2WriteEnabled=true
    Module.Param3WriteEnabled=true
   )
  ]
  rendererClasses [&NiagaraSpriteRendererProperties]
  renderers [{rendererClass=&NiagaraSpriteRendererProperties}]
 }
 Darken002 {
  emitterUpdateScript [
   &EmitterState(Life Cycle Mode=::NewEnumerator1, Loop Behavior=::NewEnumerator1, Loop Duration Mode=::NewEnumerator1, UseLoopDelay=false)
   &SpawnBurst_Instantaneous(Spawn Count=1, Use Spawn Probability=false)
  ]
  particleSpawnScript [
   &InitializeParticle(
    Color=<rgba 0 0 0 1>
    Module.Write Color=true
    Module.Write Lifetime=false
    Module.Write Mass=true
    Module.Write Material Random=true
    Module.Write Position=true
    Module.Write Scale=false
    Module.Write SpriteRotation=false
    Module.Write SpriteSize=true
    Position=&Add_Vector(A=&SimulationPosition(), B=<0 -8.5 0>)
    Sprite Size=<30 30>
   )
   &CameraOffset(Camera Offset Amount=10)
   &FN_SpriteAlignmentAndFacing(SpriteAlignment:Vector3f=<0 0 1>, SpriteFacing:Vector3f=<0 -1 0>) {module=SpriteAlignmentAndFacing}
  ]
  particleUpdateScript [
   &ParticleState(Kill Particles When Lifetime Has Elapsed=false, Let Infinitely Lived Particles Die When Emitter Deactivates=true)
   &DynamicMaterialParameters(
    Index 0 Param 1=2
    Index 0 Param 2=8
    Index 0 Param 3=2
    Index 0 Param 4=&MakeCustomFloatFromBool(Bool=$User.bIsFrontEndPreview, True Float=$System.AutoSwitch, False Float=$User.KillingChange)
    Index 3 Param 1=$User.AnimatedHologram
    Index 3 Param 2=$User.AnimationPhase
    Module.Param0WriteEnabled=true
    Module.Param0WriteEnabled3=true
    Module.Param1WriteEnabled=true
    Module.Param1WriteEnabled3=true
    Module.Param2WriteEnabled=true
    Module.Param2WriteEnabled3=true
    Module.Param3WriteEnabled=true
    Module.Param3WriteEnabled3=true
    Write Parameter Index 3=true
   )
  ]
  rendererClasses [&NiagaraSpriteRendererProperties]
  renderers [{rendererClass=&NiagaraSpriteRendererProperties}]
 }
 Sticker {
  emitterUpdateScript [
   &EmitterState(Life Cycle Mode=::NewEnumerator1, Loop Duration=1, UseLoopDelay=false)
   &SpawnBurst_Instantaneous(Spawn Count=1, Use Spawn Probability=false)
   /SetVariables(Emitter.FEScale:NiagaraFloat=&MakeCustomFloatFromBool(Bool=$User.bIsFrontEndPreview, True Float=0.5, False Float=1)) {isSetParameters=true}
  ]
  particleSpawnScript [
   &InitializeParticle(
    Module.Write Color=true
    Module.Write Lifetime=true
    Module.Write Mass=true
    Module.Write Material Random=true
    Module.Write Position=true
    Module.Write Scale=true
    Module.Write SpriteRotation=false
    Module.Write SpriteSize=false
    Position=&Add_Vector(A=&SimulationPosition(), B=<0 -8 0>)
    Sprite Size=<5 5>
   )
   /Niagara/Modules/Spawn/Orientation/V2/InitialMeshOrientation(Rotation=<0 0 -0.25>, Use Orientation Vector=false, Use Rotation Vector=true)
   &CameraOffset(Camera Offset Amount=10)
  ]
  particleUpdateScript [
   &ParticleState()
   &SolveForcesAndVelocity(Clamp Velocity=false, Limit Acceleration=false)
   &DynamicMaterialParameters(
    Index 0 Param 1=&Waveform(
     "Waveform [1]"=::NewEnumerator8
     "[1] Phase Scale"=20
     "[1] Frequency"=1
     "[1] Amplitude Min/Max"=<0 1>
     "[1] Scale"=0.95
     "[1] Bias"=0.05
     "[1] Exponent"=1.3
     "[2] Phase"=$Emitter.Age
     "[3] Phase"=$Emitter.Age
    )
    Index 0 Param 2=&Waveform(
     "Waveform [1]"=::NewEnumerator8
     "[1] Phase Scale"=25
     "[1] Frequency"=1
     "[1] Amplitude Min/Max"=<0 1>
     "[1] Scale"=0.7
     "[1] Bias"=0.3
     "[2] Phase"=$Emitter.Age
     "[3] Phase"=$Emitter.Age
    )
    Index 0 Param 3=&MakeCustomFloatFromBool(Bool=$User.bIsFrontEndPreview, True Float=$System.AutoSwitch, False Float=$User.KillingChange)
    Index 1 Param 1=$Emitter.FEScale
    Index 3 Param 1=$User.AnimatedHologram
    Index 3 Param 2=$User.AnimationPhase
    Module.Param0WriteEnabled=true
    Module.Param0WriteEnabled1=true
    Module.Param0WriteEnabled3=true
    Module.Param1WriteEnabled=true
    Module.Param1WriteEnabled1=true
    Module.Param1WriteEnabled3=true
    Module.Param2WriteEnabled=true
    Module.Param2WriteEnabled1=true
    Module.Param2WriteEnabled3=true
    Module.Param3WriteEnabled=true
    Module.Param3WriteEnabled1=true
    Module.Param3WriteEnabled3=true
    Write Parameter Index 1=true
    Write Parameter Index 3=true
   )
   &ScaleColor(Scale RGB=&VectorFromFloat(Value=$User.Boost), ScaleA=true, ScaleRGB=true)
  ]
  rendererClasses [&NiagaraMeshRendererProperties]
  renderers [
   {rendererClass=&NiagaraMeshRendererProperties, rendererIndex=0}
   {rendererClass=&NiagaraMeshRendererProperties, rendererIndex=1}
  ]
 }
}
summary {
 systemName=NS_ZombieElastic_Backpack
 emitters [
  {emitterName=Sticker, rendererClasses [&NiagaraMeshRendererProperties]}
  {emitterName=Darken, rendererClasses [&NiagaraSpriteRendererProperties]}
  {emitterName=Darken001, rendererClasses [&NiagaraSpriteRendererProperties]}
  {emitterName=Darken002, rendererClasses [&NiagaraSpriteRendererProperties]}
 ]
 userVariables [
  User.AnimatedHologram {type=NiagaraFloat, defaultValue=1, description=""}
  User.AnimationPhase {type=NiagaraFloat, defaultValue=0, description=""}
  User.bIsFrontEndPreview {type=NiagaraBool}
  User.Boost {type=NiagaraFloat, defaultValue=1, description=""}
  User.HologramTexture {type=Texture, description="", defaultValue {struct=NiagaraExt_VariableValue_Object, value {object=null, objectClass=/Script/Engine.Texture}}}
  User.KillingChange {type=NiagaraFloat, defaultValue=-0.1, description=""}
  User.MID_ElasticHologram {
   type=MaterialInterface
   description=""
   defaultValue {struct=NiagaraExt_VariableValue_Object, value {object=null, objectClass=/Script/Engine.MaterialInterface}}
  }
 ]
}
